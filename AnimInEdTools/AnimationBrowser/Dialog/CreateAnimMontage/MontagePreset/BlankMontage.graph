EditorUtilityBlueprint /AnimInEdTools/AnimationBrowser/Dialog/CreateAnimMontage/MontagePreset/BlankMontage
graphs [
 Create Montage {nodes [
   n1 = Create Montage!() -> (then, AssetName:String, PackagePath:String, AnimSequence:Anim Sequence Object Reference) {is=title K2Node_FunctionEntry}
   > n2 = Execute Python Script!(
    execute
    PythonScript="import unreal\r\nfactory = unreal.AnimMontageFactory()\r\nprint('factory', factory)\r\nprint('source_animation', source_animation)\r\nprint(dir(factory.source_animation))\r\nanim_montage = None\r\n'''\r\nfactory.source_animation(source_animation)\r\n\r\n## Get asset tools\r\nasset_tools = unreal.AssetToolsHelpers.get_asset_tools()\r\n \r\nanim_montage = unreal.AssetTools.create_asset(asset_tools, asset_name = asset_name, package_path = package_path, asset_class = unreal.LevelSequence, factory = factory)\r\n'''"
    source_animation:Anim Sequence Object Reference=@n1.AnimSequence
    asset_name:String=@n1
    package_path:String=@n1.PackagePath
   ) -> (then, ReturnValue, factory:Anim Montage Factory Object Reference) {inputs [source_animation, asset_name, package_path], outputs [factory]}
   > Set SourceAnimation!(execute, SourceAnimation:Anim Sequence Object Reference=@n1.AnimSequence, self:Anim Montage Factory Object Reference=@n2.factory) -> (then, Output_Get:Anim Sequence Object Reference)
   > n3 = GetAssetTools()
   > n4 = CreateAsset(self=@n3, AssetName=@n1, PackagePath=@n1.PackagePath, AssetClass=/Script/Engine.AnimMontage, Factory=@n2.factory)
   > n5 = Cast To AnimMontage!(execute, Object=@n4) -> (then, CastFailed, AsAnim Montage:Anim Montage Object Reference) {is=title K2Node_DynamicCast}
   > Set New AnimMontage(New AnimMontage:Anim Montage Object Reference=@n5) -> (Output_Get:Anim Montage Object Reference)
   > Return Node!(execute, AnimMontage:Anim Montage Object Reference=Get New AnimMontage() -> (New AnimMontage:Anim Montage Object Reference))
   Return Node(execute=@n5.CastFailed)
  ]}
 EventGraph {links [], nodes []}
]
variables []
